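# Sound cue for the destruction sound of the Bass_A prop.
# A Random node picks between two destruction wave variations with equal weight.
# Playback uses Fortnite's medium-object demolition attenuation, the impacts sound class and a concurrency set for object impacts, with a boosted volume multiplier.
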SoundCue /Hera/Audio/Props/Bass_A/SC_Bass_A_Desctruction
root=Random_0
nodes [
 {id=Random_0, children [WavePlayer_0,WavePlayer_1], properties {weights [1,1]}}
 {id=WavePlayer_0, properties {soundWaveAssetPtr=&/Bass_A_Destruction_01}}
 {id=WavePlayer_1, properties {soundWaveAssetPtr=&/Bass_A_Destruction_02}}
]
properties {
 attenuationSettings=/Game/Sounds/Attenuation/Fort_Demolition_Destroy_Medium_Sized_Object_Att
 maxDistance=4000
 soundClassObject=&Fort_Impacts_PSM
 volumeMultiplier=5
 concurrencySet [&CONC_Object_Impacts_FourMax]
}
